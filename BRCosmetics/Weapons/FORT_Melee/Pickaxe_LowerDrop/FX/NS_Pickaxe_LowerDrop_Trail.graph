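# Niagara swing trail for the LowerDrop pickaxe cosmetic, combining animation-trail ribbons with socket-spawned sprite particles.
# Main and Yellow ribbon emitters cover head and handle between indexed bones or sockets, with the Main pair adding multicolor curves and separate line color and width attributes.
# Swing particles spawn by rate and movement, with quality-scaled spacing, front-end size and speed variants, and soft-glow attributes.
# System-level time-of-day color tints all emitters, while AlphaFade scales the ribbons.
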NiagaraSystem /BRCosmetics/Weapons/FORT_Melee/Pickaxe_LowerDrop/FX/NS_Pickaxe_LowerDrop_Trail
dependencies {
 usedDataInterfaces [
  &NiagaraDataInterfaceArrayFloat
  &NiagaraDataInterfaceCurve
  &NiagaraDataInterfaceVector2DCurve
  &NiagaraDataInterfaceColorCurve
  &NiagaraDataInterfaceSocketReader
  &NiagaraDataInterfaceArrayColor
 ]
 usedDynamicInputs [
  &Vector2DFromCurve
  &SelectLinearColorFromArray
  &MergeHelper_FloatSwitch
 ]
 usedModules [
  &self:SystemSpawnScript.SystemScriptSource.SystemScriptGraph.NiagaraNodeAssignment_1.SetVariables_#1
  &self:SystemSpawnScript.SystemScriptSource.SystemScriptGraph.NiagaraNodeAssignment_5.SetVariables_#2
  &SystemState
  &self:SystemSpawnScript.SystemScriptSource.SystemScriptGraph.NiagaraNodeAssignment_3.SetVariables_#3
  &self:SystemSpawnScript.SystemScriptSource.SystemScriptGraph.NiagaraNodeAssignment_0.SetVariables_#4
  &self:SystemSpawnScript.SystemScriptSource.SystemScriptGraph.NiagaraNodeAssignment_2.SetVariables_#5
  &self:SystemSpawnScript.SystemScriptSource.SystemScriptGraph.NiagaraNodeAssignment_4.SetVariables_#6
 ]
}
emitters {
 Handle_Trail_Main {
  emitterSpawnScript [&FN_RibbonIDOffsetUtility()]
  emitterUpdateScript [
   &EmitterState()
   &FN_SpawnAnimTrail(Spawn Interval=0)
  ]
  particleSpawnScript [
   &FN_InitializeAnimTrailBoneIndecies(First Bone or Socket Index=3, Second Bone or Socket Index=2)
   &FN_InitializeAnimTrail(
    AnimTrailWidth=1.2
    Color=&Multiply_LinearColorByFloat(Linear Color=$System.Initial Random Color, Float=$System.AlphaFade)
    Lifetime=0.3
   )
  ]
  particleUpdateScript [
   &FN_UpdateAnimTrail() {module=FN_UpdateAnimTrail001}
   &SolveForcesAndVelocity(Clamp Velocity=false, Limit Acceleration=false)
   &ScaleColor(
    Scale Alpha=&FloatFromCurve(
     FloatCurve=DataInterface() {propertyValues {Curve {keys [{arriveTangent=-2.2326, interpMode=RCIM_Cubic, leaveTangent=-2.2326, tangentMode=RCTM_User, time=0, value=1}, <at 0.5 0>]}}}
     Scale Curve=&Multiply_Float(A=0.2, B=$System.AlphaFade)
    )
    Scale RGB=&Multiply_Vector(
     A=&MakeVectorFromLinearColorRGB(Color=$System.TODColorMultiplier)
     B=&Multiply_VectorByFloat(
      Vector=&MakeVectorFromLinearColorRGB(Color=&ColorFromCurve(ColorCurve=DataInterface() {propertyValues {
          AlphaCurve {keys [<at 0 1 RCIM_Cubic>]}
          BlueCurve {keys [
            <at 0.0813 1 RCIM_Cubic>
            <at 0.2728 1>
            <at 0.4224 0.7983>
            {arriveTangent=-0.4977, interpMode=RCIM_Cubic, leaveTangent=-0.4977, time=0.6042, value=0.5576}
            <at 0.8468 0.5871>
           ]}
          GreenCurve {keys [
            <at 0.0813 0.8879 RCIM_Cubic>
            <at 0.2728 0.1812>
            <at 0.4224 0.1178>
            {arriveTangent=0.0625, interpMode=RCIM_Cubic, leaveTangent=0.0625, time=0.6042, value=0.5871}
            <at 0.8468 0.1443>
           ]}
          RedCurve {keys [
            <at 0.0813 0.159 RCIM_Cubic>
            <at 0.2728 0.0545>
            <at 0.4224 0.8469>
            {arriveTangent=-1.9043, interpMode=RCIM_Cubic, leaveTangent=-1.9043, time=0.6042, value=0.2599}
            <at 0.8468 0.0387>
           ]}
         }}))
      Float=2
     )
    )
    ScaleA=true
    ScaleRGB=true
   )
   /SetVariables(
    Particles.Color_Lines:LinearColor=&MultiplyLinearColor(
     Linear Color=$Particles.Initial.Color
     Scale Factor=&MakeVector4FromVectorAndFloat(
      "Vector (RGB)"=&Multiply_Vector(
       A=&MakeVectorFromLinearColorRGB(Color=$System.TODColorMultiplier)
       B=&Multiply_VectorByFloat(
        Vector=&MakeVectorFromLinearColorRGB(Color=&MultiplyLinearColor(
          Linear Color=&ColorFromCurve(ColorCurve=DataInterface() {propertyValues {
             AlphaCurve {keys [<at 0 1 RCIM_Cubic>]}
             BlueCurve {keys [<at 0.0582 1 RCIM_Cubic>,<at 0.2637 1 RCIM_Cubic>,<at 0.7315 1>]}
             GreenCurve {keys [<at 0.0582 0.0953 RCIM_Cubic>,<at 0.2637 0.8879 RCIM_Cubic>,<at 0.7315 0.1812>]}
             RedCurve {keys [<at 0.0582 0.9131 RCIM_Cubic>,<at 0.2637 0.159 RCIM_Cubic>,<at 0.7315 0.0545>]}
            }})
          Scale Factor=&MakeVector4_FromFloat(Float=&FloatFromCurve(FloatCurve=DataInterface() {propertyValues {Curve {keys [<at 0 5>,<at 0.35 0.5>]}}}))
         ))
        Float=1
       )
      )
      "Float (Alpha)"=&FloatFromCurve(
       FloatCurve=DataInterface() {propertyValues {Curve {keys [
           {arriveTangent=4.8178, interpMode=RCIM_Cubic, leaveTangent=4.8178, tangentMode=RCTM_User, time=0, value=0.3}
           {arriveTangent=-3.2686, interpMode=RCIM_Cubic, leaveTangent=-3.2686, tangentMode=RCTM_User, time=0.1354, value=0.9529}
           {tangentMode=RCTM_User, time=0.5, value=0}
          ]}}}
       Scale Curve=&Multiply_Float(A=$System.AlphaFade, B=0.25)
      )
     )
    )
    Particles.Color_Tip:LinearColor=&MultiplyLinearColor(Linear Color=$Particles.Color, Scale Factor=<1 1 1 4>)
   ) {isSetParameters=true, module=SetVariables_#7}
   /SetVariables(Particles.RibbonScaleLines:NiagaraFloat=&Multiply_Float(
     A=&FloatFromCurve(
      FloatCurve=DataInterface() {propertyValues {Curve {keys [<at 0 0.8 RCIM_Cubic>, {arriveTangent=-1.0934, interpMode=RCIM_Cubic, leaveTangent=-1.0934, tangentMode=RCTM_User, time=1, value=0.25}]}}}
      Scale Curve=1.1
     )
     B=$Particles.RibbonWidth
    )) {isSetParameters=true, module=SetVariables_#8}
   &DynamicMaterialParameters(
    Module.Index1Float0Write=false
    Module.Index1Float1Write=false
    Module.Index1Float2Write=false
    Module.Index1Float3Write=true
    Write Parameter Index 0=::NewEnumerator0
    Write Parameter Index 1=::NewEnumerator1
   )
   &ScaleColor(
    Scale Alpha=&FloatFromCurve(
     FloatCurve=DataInterface() {propertyValues {Curve {defaultValue=1, keys [<at 0 0 RCIM_Cubic>, {interpMode=RCIM_Cubic, tangentMode=RCTM_User, time=0.1, value=1}]}}}
     Scale Curve=0.25
    )
    ScaleA=true
    ScaleRGB=true
   ) {module=ScaleColor001}
  ]
  rendererClasses [&NiagaraRibbonRendererProperties]
  renderers [
   {rendererClass=&NiagaraRibbonRendererProperties}
   {rendererClass=&NiagaraRibbonRendererProperties, rendererIndex=1}
  ]
 }
 Head_Trail_Main {
  emitterSpawnScript [&FN_RibbonIDOffsetUtility()]
  emitterUpdateScript [
   &EmitterState()
   &FN_SpawnAnimTrail(Spawn Interval=0)
  ]
  particleSpawnScript [
   &FN_InitializeAnimTrailBoneIndecies(Second Bone or Socket Index=0)
   &FN_InitializeAnimTrail(
    AnimTrailWidth=30
    Color=&Multiply_LinearColorByFloat(Linear Color=$System.Initial Random Color, Float=$System.AlphaFade)
    Lifetime=0.3
   )
  ]
  particleUpdateScript [
   &FN_UpdateAnimTrail() {module=FN_UpdateAnimTrail001}
   &SolveForcesAndVelocity(Clamp Velocity=false, Limit Acceleration=false)
   &ScaleColor(
    Scale Alpha=&FloatFromCurve(
     FloatCurve=DataInterface() {propertyValues {Curve {keys [{arriveTangent=-2.2326, interpMode=RCIM_Cubic, leaveTangent=-2.2326, tangentMode=RCTM_User, time=0, value=1}, <at 0.5 0>]}}}
     Scale Curve=&Multiply_Float(A=0.3, B=$System.AlphaFade)
    )
    Scale RGB=&Multiply_Vector(
     A=&MakeVectorFromLinearColorRGB(Color=$System.TODColorMultiplier)
     B=&Multiply_VectorByFloat(
      Vector=&MakeVectorFromLinearColorRGB(Color=&ColorFromCurve(ColorCurve=DataInterface() {propertyValues {
          AlphaCurve {keys [<at 0 1 RCIM_Cubic>]}
          BlueCurve {keys [
            <at 0.0813 1 RCIM_Cubic>
            <at 0.2728 1>
            <at 0.4224 0.7983>
            {arriveTangent=-0.4977, interpMode=RCIM_Cubic, leaveTangent=-0.4977, time=0.6042, value=0.5576}
            <at 0.8468 0.5871>
           ]}
          GreenCurve {keys [
            <at 0.0813 0.8879 RCIM_Cubic>
            <at 0.2728 0.1812>
            <at 0.4224 0.1178>
            {arriveTangent=0.0625, interpMode=RCIM_Cubic, leaveTangent=0.0625, time=0.6042, value=0.5871}
            <at 0.8468 0.1443>
           ]}
          RedCurve {keys [
            <at 0.0813 0.159 RCIM_Cubic>
            <at 0.2728 0.0545>
            <at 0.4224 0.8469>
            {arriveTangent=-1.9043, interpMode=RCIM_Cubic, leaveTangent=-1.9043, time=0.6042, value=0.2599}
            <at 0.8468 0.0387>
           ]}
         }}))
      Float=1.4
     )
    )
    ScaleA=true
    ScaleRGB=true
   )
   /SetVariables(
    Particles.Color_Lines:LinearColor=&MultiplyLinearColor(
     Linear Color=$Particles.Initial.Color
     Scale Factor=&MakeVector4FromVectorAndFloat(
      "Vector (RGB)"=&Multiply_Vector(
       A=&MakeVectorFromLinearColorRGB(Color=$System.TODColorMultiplier)
       B=&Multiply_VectorByFloat(
        Vector=&MakeVectorFromLinearColorRGB(Color=&MultiplyLinearColor(
          Linear Color=&ColorFromCurve(ColorCurve=DataInterface() {propertyValues {
             AlphaCurve {keys [<at 0 1 RCIM_Cubic>]}
             BlueCurve {keys [<at 0.0582 1 RCIM_Cubic>,<at 0.2637 1 RCIM_Cubic>,<at 0.7315 1>]}
             GreenCurve {keys [<at 0.0582 0.0953 RCIM_Cubic>,<at 0.2637 0.8879 RCIM_Cubic>,<at 0.7315 0.1812>]}
             RedCurve {keys [<at 0.0582 0.9131 RCIM_Cubic>,<at 0.2637 0.159 RCIM_Cubic>,<at 0.7315 0.0545>]}
            }})
          Scale Factor=&MakeVector4_FromFloat(Float=&FloatFromCurve(FloatCurve=DataInterface() {propertyValues {Curve {keys [<at 0 5>,<at 0.35 0.5>]}}}))
         ))
        Float=1
       )
      )
      "Float (Alpha)"=&FloatFromCurve(
       FloatCurve=DataInterface() {propertyValues {Curve {keys [
           {arriveTangent=4.8178, interpMode=RCIM_Cubic, leaveTangent=4.8178, tangentMode=RCTM_User, time=0, value=0.3}
           {arriveTangent=-3.2686, interpMode=RCIM_Cubic, leaveTangent=-3.2686, tangentMode=RCTM_User, time=0.1354, value=0.9529}
           {tangentMode=RCTM_User, time=0.5, value=0}
          ]}}}
       Scale Curve=$System.AlphaFade
      )
     )
    )
    Particles.Color_Tip:LinearColor=&MultiplyLinearColor(Linear Color=$Particles.Color, Scale Factor=<1 1 1 4>)
   ) {isSetParameters=true, module=SetVariables_#7}
   /SetVariables(Particles.RibbonScaleLines:NiagaraFloat=&Multiply_Float(
     A=&FloatFromCurve(
      FloatCurve=DataInterface() {propertyValues {Curve {keys [<at 0 0.8 RCIM_Cubic>, {arriveTangent=-1.0934, interpMode=RCIM_Cubic, leaveTangent=-1.0934, tangentMode=RCTM_User, time=1, value=0.25}]}}}
      Scale Curve=1.1
     )
     B=$Particles.RibbonWidth
    )) {isSetParameters=true, module=SetVariables_#8}
   &DynamicMaterialParameters(
    Module.Index1Float0Write=false
    Module.Index1Float1Write=false
    Module.Index1Float2Write=false
    Module.Index1Float3Write=true
    Write Parameter Index 0=::NewEnumerator0
    Write Parameter Index 1=::NewEnumerator1
   )
   &ScaleColor(
    Scale Alpha=&FloatFromCurve(FloatCurve=DataInterface() {propertyValues {Curve {defaultValue=1, keys [<at 0 0 RCIM_Cubic>, {interpMode=RCIM_Cubic, tangentMode=RCTM_User, time=0.1, value=1}]}}})
    ScaleA=true
    ScaleRGB=true
   ) {module=ScaleColor001}
  ]
  rendererClasses [&NiagaraRibbonRendererProperties]
  renderers [
   {rendererClass=&NiagaraRibbonRendererProperties}
   {rendererClass=&NiagaraRibbonRendererProperties, rendererIndex=1}
  ]
 }
 Particles_Swing002 {
  emitterUpdateScript [
   &EmitterState(Life Cycle Mode=::NewEnumerator1, Loop Behavior=::NewEnumerator1, Loop Delay=0.05, Loop Duration=0.3, UseLoopDelay=true)
   &SpawnRate(SpawnRate=55, Use Spawn Probability=false)
   &SpawnPerUnit(
    Max Movement Threshold=2500
    Module.Use Max Movement Threshold=true
    Module.Use Movement Tolerance=false
    Spawn Spacing:NiagaraFloat=&SelectFloatFromArray(Direct Array Index=$Engine.QualityLevel, Float Selection Array=DataInterface() {propertyValues {FloatData [50,50,40,30,20]}})
    Use Spawn Probability=false
   )
  ]
  particleSpawnScript [
   &InitializeParticle(
    Color=&MultiplyLinearColor(
     Linear Color=<rgba 5 5 5 2>
     Scale Factor=&MakeVector4_FromLinearColor(LinearColor=$System.TODColorMultiplier)
    )
    Lifetime Max=0.45
    Lifetime Min=0.166
    Lifetime Mode=::NewEnumerator1
    Sprite Rotation Mode=::NewEnumerator2
    Sprite Size Mode=::NewEnumerator3
    Uniform Sprite Size=&Multiply_Float(
     A=&MakeCustomFloatFromBool(Bool=$User.bIsFrontEnd, True Float=0.45, False Float=1)
     B=&RandomRangeFloat(Minimum=1, Maximum=2)
    )
    UsePositionOffset=false
    Write Material Random=true
   )
   &SocketLocation(SamplePosOffset=false, Socket Reader=$System.Socket Reader, Socket Sampling Mode=::NewEnumerator3, VelocitySpeedLimit=false)
   &ShapeLocation(MaskSpawnGroup=false, Override Seed=false, Sphere Radius=9, Sphere Surface Distribution=1, UseOwnerScale=true)
   &AddVelocity(
    Module.Use Falloff Distance=false
    Override Seed=false
    Velocity Mode=::NewEnumerator1
    Velocity Speed=&FN_FrontEndFloatModifier(
     Gameplay=&RandomRangeFloat(Minimum=30, Maximum=75)
     Front End=&RandomRangeFloat(Minimum=150, Maximum=400)
    )
   )
   &InheritVelocity(Inherited Velocity Speed Limit=20, Speed Limit=true, UseSpeedThreshold=false, Velocity Source=$Output.SocketLocation.SampledVelocity)
   &DynamicMaterialParameters(
    Index 0 Param 1=3
    Index 0 Param 2=20
    Index 0 Param 3=1
    Module.Index0Float0Write=true
    Module.Index0Float1Write=true
    Module.Index0Float2Write=true
    Module.Index0Float3Write=true
   )
  ]
  particleUpdateScript [
   &ParticleState()
   &GravityForce(Gravity=<0 0 -200>)
   &Drag(Drag=&UniformRangedFloat(Minimum=2, Maximum=6), Use Linear Drag=true, Use Rotational Drag=false)
   &SolveForcesAndVelocity(Clamp Velocity=false, Limit Acceleration=false)
   &ScaleSpriteSize(Uniform Curve Sprite Scale=DataInterface() {propertyValues {Curve {
       defaultValue=0
       keys [
        {arriveTangent=9.0725, interpMode=RCIM_Cubic, leaveTangent=9.0725, tangentMode=RCTM_User, time=0, value=0}
        {arriveTangent=-2.5249, interpMode=RCIM_Cubic, leaveTangent=-2.5249, tangentMode=RCTM_User, time=0.1408, value=0.9847}
        {arriveTangent=-0.0405, interpMode=RCIM_Cubic, leaveTangent=-0.0405, tangentMode=RCTM_User, time=1, value=0}
       ]
      }}})
   &ScaleColor(
    Scale Mode=::NewEnumerator1
    Scale RGBA=&MakeVector4_FromLinearColor(LinearColor=&RandomRangeLinearColor(
      Minimum=&ColorFromCurve(ColorCurve=DataInterface() {propertyValues {
         AlphaCurve {keys [<at 0.333 1>,<at 1 0>]}
         BlueCurve {keys [<at 0 1>,<at 0.166 5>,<at 0.333 4>,<at 0.75 3>]}
         GreenCurve {keys [<at 0 0.24>,<at 0.166 0.2>,<at 0.333 0>,<at 0.75 0>]}
         RedCurve {keys [<at 0 0.9367>,<at 0.166 3.4>,<at 0.333 2.3333>,<at 0.75 1.25>]}
        }})
      Maximum=&ColorFromCurve(ColorCurve=DataInterface() {propertyValues {
         AlphaCurve {keys [<at 0.333 1>,<at 1 0>]}
         BlueCurve {keys [<at 0 1>,<at 0.166 5>,<at 0.333 4>,<at 0.75 3>]}
         GreenCurve {keys [<at 0 0.8987>,<at 0.166 3.8>,<at 0.333 2>,<at 0.75 0.9>]}
         RedCurve {keys [<at 0 0.24>,<at 0.166 0.2>,<at 0.333 0>,<at 0.75 0>]}
        }})
     ))
    ScaleRGBA=true
   )
   /SetVariables(
    Particles.SoftGlowColor:LinearColor=&Multiply_LinearColorByFloat(Linear Color=$Particles.Color, Float=0.07)
    Particles.SoftGlowSize:Vector2f=&Multiply_Vector2D_ByFloat(Float=6, Vector2D=$Particles.SpriteSize)
   ) {isSetParameters=true}
  ]
  rendererClasses [&NiagaraSpriteRendererProperties]
  renderers [
   {rendererClass=&NiagaraSpriteRendererProperties}
   {rendererClass=&NiagaraSpriteRendererProperties, rendererIndex=1}
  ]
 }
 Yellow_Trail_Handle {
  emitterSpawnScript [&FN_RibbonIDOffsetUtility()]
  emitterUpdateScript [
   &EmitterState()
   &FN_SpawnAnimTrail()
  ]
  particleSpawnScript [
   &FN_InitializeAnimTrailBoneIndecies(First Bone or Socket Index=1, Second Bone or Socket Index=2)
   &FN_InitializeAnimTrail(Color=<rgba 0.6609 0.6609 0.6609 1>, Lifetime=0.3)
  ]
  particleUpdateScript [
   &SolveForcesAndVelocity(Clamp Velocity=false, Limit Acceleration=false)
   &FN_UpdateAnimTrail() {module=FN_UpdateAnimTrail001}
   &DynamicMaterialParameters(
    Index 1 Param 3=&FloatFromCurve(
     FloatCurve=DataInterface() {propertyValues {Curve {keys [<at 0 0>, {arriveTangent=1.9814, tangentMode=RCTM_User, time=1, value=1}]}}}
     CurveIndex=&ModuloFloat(A=$Emitter.Age, B=1)
     Scale Curve=0
    )
    Index 1 Param 4=1
    Index 2 Param 1=10
    Index 2 Param 2=1
    Index 2 Param 3=0.5
    Index 2 Param 4=0.6
    Module.Index0Float0Write=true
    Module.Index0Float1Write=true
    Module.Index0Float2Write=true
    Module.Index0Float3Write=true
    Module.Index1Float0Write=true
    Module.Index1Float1Write=true
    Module.Index1Float2Write=true
    Module.Index1Float3Write=true
    Module.Index2Float0Write=true
    Module.Index2Float1Write=true
    Module.Index2Float2Write=true
    Module.Index2Float3Write=true
    Write Parameter Index 1=::NewEnumerator1
    Write Parameter Index 2=::NewEnumerator1
   )
   &ScaleColor(
    Scale Alpha=&FloatFromCurve(
     FloatCurve=DataInterface() {propertyValues {Curve {keys [{arriveTangent=-2.2326, interpMode=RCIM_Cubic, leaveTangent=-2.2326, tangentMode=RCTM_User, time=0, value=1}, <at 0.75 0>]}}}
     Scale Curve=&Multiply_Float(A=0.25, B=$System.AlphaFade)
    )
    Scale RGB=&Multiply_Vector(A=&MakeVectorFromLinearColorRGB(Color=$System.TODColorMultiplier), B=<1 1 1>)
    ScaleA=true
    ScaleRGB=true
   ) {module=ScaleColor001}
   &ScaleColor(
    Scale Alpha=&FloatFromCurve(FloatCurve=DataInterface() {propertyValues {Curve {defaultValue=1, keys [<at 0 0 RCIM_Cubic>, {interpMode=RCIM_Cubic, tangentMode=RCTM_User, time=0.1, value=1}]}}})
    ScaleA=true
    ScaleRGB=true
   ) {module=ScaleColor002}
  ]
  rendererClasses [&NiagaraRibbonRendererProperties]
  renderers [
   {rendererClass=&NiagaraRibbonRendererProperties}
   {rendererClass=&NiagaraRibbonRendererProperties, rendererIndex=1}
  ]
 }
 Yellow_Trail_Head {
  emitterSpawnScript [&FN_RibbonIDOffsetUtility()]
  emitterUpdateScript [
   &EmitterState()
   &FN_SpawnAnimTrail()
  ]
  particleSpawnScript [
   &FN_InitializeAnimTrailBoneIndecies(Second Bone or Socket Index=0)
   &FN_InitializeAnimTrail(AnimTrailWidth=27, Color=<rgba 0.6609 0.6609 0.6609 1>, Lifetime=0.3)
  ]
  particleUpdateScript [
   &SolveForcesAndVelocity(Clamp Velocity=false, Limit Acceleration=false)
   &FN_UpdateAnimTrail() {module=FN_UpdateAnimTrail001}
   &DynamicMaterialParameters(
    Index 1 Param 3=&FloatFromCurve(
     FloatCurve=DataInterface() {propertyValues {Curve {keys [<at 0 0>, {arriveTangent=1.9814, tangentMode=RCTM_User, time=1, value=1}]}}}
     CurveIndex=&ModuloFloat(A=$Emitter.Age, B=1)
     Scale Curve=0
    )
    Index 1 Param 4=1
    Index 2 Param 1=10
    Index 2 Param 2=1
    Index 2 Param 3=0.5
    Index 2 Param 4=0.6
    Module.Index0Float0Write=true
    Module.Index0Float1Write=true
    Module.Index0Float2Write=true
    Module.Index0Float3Write=true
    Module.Index1Float0Write=true
    Module.Index1Float1Write=true
    Module.Index1Float2Write=true
    Module.Index1Float3Write=true
    Module.Index2Float0Write=true
    Module.Index2Float1Write=true
    Module.Index2Float2Write=true
    Module.Index2Float3Write=true
    Write Parameter Index 1=::NewEnumerator1
    Write Parameter Index 2=::NewEnumerator1
   )
   &ScaleColor(
    Scale Alpha=&FloatFromCurve(
     FloatCurve=DataInterface() {propertyValues {Curve {keys [{arriveTangent=-2.2326, interpMode=RCIM_Cubic, leaveTangent=-2.2326, tangentMode=RCTM_User, time=0, value=1}, <at 0.75 0>]}}}
     Scale Curve=&Multiply_Float(A=0.25, B=$System.AlphaFade)
    )
    Scale RGB=&Multiply_Vector(A=&MakeVectorFromLinearColorRGB(Color=$System.TODColorMultiplier), B=<1 1 1>)
    ScaleA=true
    ScaleRGB=true
   ) {module=ScaleColor001}
   &ScaleColor(
    Scale Alpha=&FloatFromCurve(FloatCurve=DataInterface() {propertyValues {Curve {defaultValue=1, keys [<at 0 0 RCIM_Cubic>, {interpMode=RCIM_Cubic, tangentMode=RCTM_User, time=0.1, value=1}]}}})
    ScaleA=true
    ScaleRGB=true
   ) {module=ScaleColor002}
  ]
  rendererClasses [&NiagaraRibbonRendererProperties]
  renderers [{rendererClass=&NiagaraRibbonRendererProperties}]
 }
}
summary {
 systemName=NS_Pickaxe_LowerDrop_Trail
 emitters [
  {emitterName=Particles_Swing002, rendererClasses [&NiagaraSpriteRendererProperties]}
  {emitterName=Yellow_Trail_Head, rendererClasses [&NiagaraRibbonRendererProperties]}
  {emitterName=Head_Trail_Main, rendererClasses [&NiagaraRibbonRendererProperties]}
  {emitterName=Handle_Trail_Main, rendererClasses [&NiagaraRibbonRendererProperties]}
  {emitterName=Yellow_Trail_Handle, rendererClasses [&NiagaraRibbonRendererProperties]}
 ]
 userVariables [
  User.bIsFrontEnd {type=NiagaraBool}
  User.bIsFrontEndPreview {type=NiagaraBool}
  User.RibbonIDOffset {type=NiagaraInt32}
  User.SwingSectionIndex {type=NiagaraInt32}
 ]
}
