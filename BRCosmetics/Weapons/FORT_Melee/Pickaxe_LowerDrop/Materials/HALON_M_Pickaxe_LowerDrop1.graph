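# Skeletal mesh master material for the LowerDrop pickaxe cosmetic, sharing parameters with the Quail Wink, Loan Float and Evoke Find pickaxe effects.
# It builds on the 2023 CharacterParent PreFX/PostFX functions, overriding emissive, base color, specular and metallic, and blending in a Subsurface shading model.
# Named reroutes carry a color-dodged gradient with panning shimmer, a flow-noise pattern, object-aligned plane coordinates and switchable reflection-vector highlights.
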Material /BRCosmetics/Weapons/FORT_Melee/Pickaxe_LowerDrop/Materials/HALON_M_Pickaxe_LowerDrop1
graph {
 properties {
  bUseMaterialAttributes=true
  bUsedWithMeshDeformer=true
  bUsedWithSkeletalMesh=true
  pixelDepthOffsetMode=PDOM_Legacy
  shadingModel=MSM_FromMaterialExpression
  usedShadingModels=Default Lit | Subsurface | Clear Coat | Subsurface Profile | Cloth
 }
 nodes [
  ActorPositionWS()
  n1 = VertexInterpolator(VS=LocalPosition() {localOrigin=InstancePreSkinning}.XYZ)
  n2 = Time()
  n3 = VectorParameter() {group=Quail Wink / Loan Float / Evoke Find, parameterName=Noise - Panning Offset, defaultValue=<rgba 0 0 -0.01 0>}
  n4 = MaterialFunctionCall(Float3=Add(A=@n1.PS, B=Multiply(A=@n2, B=@n3.RGB))) {materialFunction=&BreakOutFloat3Components}
  n5 = Multiply(
   A=Multiply(A=NamedRerouteUsage(), B=ScalarParameter() {group=Quail Wink / Loan Float / Evoke Find, parameterName=Noise - Random Offset Multiplier})
   B=ScalarParameter() {defaultValue=1000, parameterName=Param_1}
  )
  Add(A=@n4.B, B=@n5)
  n6 = VectorParameter() {group=Quail Wink / Loan Float / Evoke Find, parameterName=Noise Tiling, defaultValue=<rgba 700 0 700 0>}
  n7 = VertexInterpolator(VS=PreSkinnedNormal())
  n8 = MaterialFunctionCall!(
   TextureObject=TextureObject() {samplerType=SAMPLERTYPE_LinearColor, texture=/Game/Global/Textures/Default/Tileable_Patterns/T_PerlinNoise_Uncompressed}
   TextureSize=@n6.RGB
   WorldPosition=MaterialFunctionCall(X=@n4.R, Y=Add(A=@n4.G, B=@n5), Z=@n4.R) {materialFunction=&MakeFloat3}.Result
   Export Float 4
   Parallax=ScalarParameter() {defaultValue=-0.1, group=Quail Wink / Loan Float / Evoke Find, parameterName=Noise - Depth}
   In
   World Space Normal=@n7.PS
   ProjectionTransitionContrast=ScalarParameter() {parameterName=Contrast}
  ) {materialFunction=/BRCosmetics/Characters/Player/Male/Medium/Bodies/M_MED_SilverBellMarine/Materials/MF_Triplanar_ScalePan_Parallax}
  n9 = TextureObject() {samplerType=SAMPLERTYPE_LinearGrayscale, texture=/BRCosmetics/Weapons/FORT_Melee/Pickaxe_QuailWink/T_ColorShift_02_Mask}
  n10 = ScalarParameter() {defaultValue=1000, parameterName=Param_2}
  n11 = ScalarParameter() {parameterName=Contrast}
  n12 = Add(
   A=Time()
   B=Multiply(
    A=ScalarParameter() {defaultValue=5, group=Quail Wink / Loan Float / Evoke Find, parameterName=Noise Modulation - Strength}
    B=MaterialFunctionCall(TextureObject=@n9, TextureSize=@n10, WorldPosition=@n1.PS, World Space Normal=@n7.PS, ProjectionTransitionContrast=@n11) {materialFunction=&WorldAlignedTexture}.XYZ Texture
   )
  )
  n13 = ScalarParameter() {defaultValue=0.05, group=Quail Wink / Loan Float / Evoke Find, parameterName=Noise Modulation - Speed}
  ComponentMask(None=Subtract(A=Power(Base=@n8.XYZ Texture) {constExponent=0.5}, B=Multiply(A=@n12, B=@n13))) {r=true}
  Fresnel() {exponent=1}
  n14 = GetMaterialAttributes(None=NamedRerouteUsage()) {attributeGetTypes [
    #1
    #2
    #3
    #4
   ]}
  n15 = NamedRerouteUsage()
  n16 = SetMaterialAttributes!(
   MaterialAttributes=@n14.MaterialAttributes
   Emissive Color=NamedRerouteUsage()
   Base Color=NamedRerouteUsage()
   Fuzz Color or Subsurface Color=NamedRerouteUsage()
   Opacity=NamedRerouteUsage()
   Specular=LinearInterpolate(A=@n14.Specular, B=ScalarParameter() {defaultValue=2, group=Quail Wink / Loan Float / Evoke Find, parameterName=Specular Blade}, Alpha=@n15) {constB=0}
   Metallic=LinearInterpolate(A=@n14.Metallic, Alpha=@n15)
  ) {attributeSetTypes [
    #1
    #2
    #5
    #6
    #3
    #4
   ]}
  n17 = ShadingModel() {shadingModel=MSM_Subsurface}
  n18 = GetMaterialAttributes(None=BlendMaterialAttributes(
    A=@n16
    B=SetMaterialAttributes!(MaterialAttributes=@n16, Shading Model=@n17) {attributeSetTypes [#7]}
    Alpha=NamedRerouteUsage()
   ))
  n19 = MaterialFunctionCall() {materialFunction=&MF_CharacterParent_2023_PreFX}
  MaterialFunctionCall(Material=@n18.MaterialAttributes, M=@n19.M, Pre Skinned Local Position=@n19.M, Distance Blend=@n19.M, Fresnel=@n19.M) {materialFunction=&MF_CharacterParent_2023_PostFX}
  MaterialFunctionCall() {materialFunction=&WorldAlignedTexture}
  n20 = ScalarParameter() {defaultValue=1.5, group=Quail Wink / Loan Float / Evoke Find, parameterName=Pattern - Min}
  Multiply(A=@n20, B=Multiply(A=NamedRerouteUsage(), B=NamedRerouteUsage()))
  n21 = Saturate(None=Fresnel(ExponentIn=Constant() {r=2}, BaseReflectFractionIn=Constant() {r=0.01}) {exponent=1})
  NamedRerouteDeclaration(None=Multiply(A=@n21, B=ScalarParameter() {defaultValue=0.7054, group=Quail Wink / Loan Float / Evoke Find, parameterName=Thickness Opacity})) {name=Edge, nodeColor=<rgba 0.1484 1 0 1>}
  n22 = MaterialFunctionCall(Float3=Transform(None=ObjectPositionWS()) {transformSourceType=TRANSFORMSOURCE_World, transformType=TRANSFORM_Instance}) {materialFunction=&BreakOutFloat3Components}
  n23 = MaterialFunctionCall(
   X=Add(A=ScalarParameter() {defaultValue=30.5806, group=Quail Wink / Loan Float / Evoke Find, parameterName="Pattern Offset - X (WS)"}, B=@n22.R)
   Y=@n22.G
   Z=Add(A=@n22.B, B=ScalarParameter() {defaultValue=14.0169, group=Quail Wink / Loan Float / Evoke Find, parameterName="Pattern Offset - Z (WS)"})
  ) {materialFunction=&MakeFloat3}
  n24 = Transform(None=Constant3Vector() {constant=<rgba 0 1 0 0>}) {transformSourceType=TRANSFORMSOURCE_Instance}
  n25 = Transform(None=Constant3Vector() {constant=<rgba 1 0 0 0>}) {transformSourceType=TRANSFORMSOURCE_Instance}
  n26 = Transform(None=Constant3Vector() {constant=<rgba 0 0 1 0>}) {transformSourceType=TRANSFORMSOURCE_Instance}
  NamedRerouteDeclaration(None=MaterialFunctionCall!(
    Projection Pivot Position=Transform(None=@n23.Result) {transformSourceType=TRANSFORMSOURCE_Instance}
    Plane Depth Offset=Constant()
    "Coordinate Scale (WS)"=ScalarParameter() {defaultValue=77.3584, group=Quail Wink / Loan Float / Evoke Find, parameterName="Pattern Scale (WS)"}
    World Space Projection Normal=Normalize(VectorInput=@n24)
    World Space Projection Tangent=Normalize(VectorInput=@n25)
    World Space Projection Binormal=Normalize(VectorInput=@n26)
   ) {materialFunction=/Engine/Functions/Engine_MaterialFunctions02/UVs/ObjectAlignedVirtualPlaneCoordinates}.Center Aligned UV Plane) {name=VIrtual Coords, nodeColor=<rgba 0 0.6953 1 1>}
  n27 = TextureCoordinate()
  n28 = ScalarParameter() {defaultValue=2.5605, group=Quail Wink / Loan Float / Evoke Find, parameterName=Flow Noise - Tiling}
  n29 = VectorParameter() {group=Quail Wink / Loan Float / Evoke Find, parameterName=Flow Noise - Speed, defaultValue=<rgba 0.05 0.1 0 0>}
  n30 = TextureSample(
   UVs=Panner(Coordinate=Multiply(A=@n27, B=@n28), Speed=@n29.RGB)
   Tex=TextureObject() {samplerType=SAMPLERTYPE_LinearColor, texture=&T_PerlinNoise_Uncompressed}
  ) {
   samplerSource=SSM_Wrap_WorldGroupSettings
   samplerType=SAMPLERTYPE_LinearColor
   texture=&T_PerlinNoise_Uncompressed
  }
  n31 = Multiply(
   A=Power(Base=@n30.R, Exp=ScalarParameter() {defaultValue=4, group=Quail Wink / Loan Float / Evoke Find, parameterName=Flow Noise - Exponent})
   B=ScalarParameter() {defaultValue=1, group=Quail Wink / Loan Float / Evoke Find, parameterName=Flow Noise - Strength}
  )
  n32 = NamedRerouteUsage()
  n33 = Step(Y=ScalarParameter() {defaultValue=0.1922, group=Quail Wink / Loan Float / Evoke Find, parameterName=Flow Mask Edge Remvoal}, X=@n32) {constY=0.8}
  NamedRerouteDeclaration(None=Multiply(A=Saturate(None=@n31), B=@n33)) {name=Pattern Flow, nodeColor=<rgba 0 1 0.9766 1>}
  n34 = LinearInterpolate(
   A=ScalarParameter() {group=Quail Wink / Loan Float / Evoke Find, parameterName=SSS Strength}
   B=ScalarParameter() {defaultValue=0.2504, group=Quail Wink / Loan Float / Evoke Find, parameterName=Blade Edge - SSS Strength}
   Alpha=NamedRerouteUsage()
  )
  NamedRerouteDeclaration(None=LinearInterpolate(B=@n34, Alpha=NamedRerouteUsage()) {constA=1, constB=0}) {name=SSS Opacity, nodeColor=<rgba 1 0.6563 0 1>}
  n35 = NamedRerouteUsage()
  n36 = NamedRerouteUsage()
  n37 = Multiply(A=NamedRerouteUsage(), B=ScalarParameter() {group=Quail Wink / Loan Float / Evoke Find, parameterName=SSS Min})
  NamedRerouteDeclaration(None=Multiply(A=Multiply(A=@n35, B=@n34), B=Max(A=@n36, B=@n37))) {name=SSS Color, nodeColor=<rgba 0 1 0.6484 1>}
  NamedRerouteDeclaration(None=Multiply(
    A=NamedRerouteDeclaration(None=@n33) {name=Pattern Mask Out, nodeColor=<rgba 0.4297 0 1 1>}
    B=ScalarParameter() {defaultValue=0.5, group=Quail Wink / Loan Float / Evoke Find, parameterName=Pattern - Opacity}
   )) {name=Pattern Alpha, nodeColor=<rgba 1 0.3281 0 1>}
  NamedRerouteDeclaration(None=@n14.Base Color) {name=Base Color In, nodeColor=<rgba 0.4766 1 0 1>}
  n38 = NamedRerouteUsage()
  n39 = ScalarParameter() {defaultValue=0.1, group=Quail Wink / Loan Float / Evoke Find, parameterName=Base Color}
  n40 = NamedRerouteUsage()
  n41 = NamedRerouteUsage()
  NamedRerouteDeclaration(None=LinearInterpolate(A=NamedRerouteUsage(), B=Multiply(A=@n38, B=@n39), Alpha=Max(A=@n40, B=@n41))) {name=Base Color, nodeColor=<rgba 1 0 0.7969 1>}
  n42 = LinearInterpolate(A=NamedRerouteUsage(), Alpha=NamedRerouteUsage()) {constB=0}
  n43 = NamedRerouteUsage()
  n44 = NamedRerouteUsage()
  n45 = Multiply(A=Multiply(A=NamedRerouteUsage(), B=@n32), B=@n20)
  n46 = Multiply(
   A=Multiply(A=NamedRerouteUsage(), B=NamedRerouteUsage())
   B=ScalarParameter() {defaultValue=10, group=Quail Wink / Loan Float / Evoke Find, parameterName=Pattern - Max}
  )
  n47 = NamedRerouteUsage()
  n48 = MaterialFunctionCall(Exponent=ScalarParameter() {defaultValue=4, group=Quail Wink / Loan Float / Evoke Find, parameterName=Pattern Fresnel}) {materialFunction=&AxisAlignedFresnel}
  NamedRerouteDeclaration(None=LinearInterpolate(A=LinearInterpolate(A=@n42, B=@n43, Alpha=@n44), B=Max(A=@n45, B=@n46), Alpha=Multiply(A=@n47, B=@n48.Fresnel))) {name=Emissive, nodeColor=<rgba 0 1 0.8828 1>}
  NamedRerouteDeclaration(None=@n19.Result) {name=Material Attributes, nodeColor=<rgba 1 0.5625 0 1>}
  NamedRerouteDeclaration(None=GetMaterialAttributes(None=@n19.Result) {attributeGetTypes [#8]}.Roughness) {name=Roughness, nodeColor=<rgba 0.1484 1 0 1>}
  NamedRerouteDeclaration(None=@n19.Tech Art Mask) {name=Debug Masks, nodeColor=<rgba 0.0312 0 1 1>}
  n49 = VertexNormalWS()
  n50 = ReflectionVectorWS(CustomWorldNormal=@n49)
  n51 = CameraVectorWS()
  n52 = StaticComponentMaskParameter(None=MaterialFunctionCall!(
    Normal=Transform(None=@n50) {transformSourceType=TRANSFORMSOURCE_World, transformType=TRANSFORM_Local}
    CameraVector=Transform(None=@n51) {transformSourceType=TRANSFORMSOURCE_World, transformType=TRANSFORM_Local}
   ) {materialFunction=/Engine/Functions/Engine_MaterialFunctions02/Utility/CustomReflectionVector}.Reflection Vector) {defaultG=true, group=Quail Wink / Loan Float / Evoke Find, parameterName=Reflection Axis}
  n53 = ReflectionVectorWS(CustomWorldNormal=@n49)
  n54 = OneMinus(None=CameraVectorWS())
  n55 = StaticComponentMaskParameter(None=MaterialFunctionCall(
    Normal=Transform(None=@n53) {transformSourceType=TRANSFORMSOURCE_World, transformType=TRANSFORM_Local}
    CameraVector=Transform(None=@n54) {transformSourceType=TRANSFORMSOURCE_World, transformType=TRANSFORM_Local}
   ) {materialFunction=&CustomReflectionVector}.Reflection Vector) {defaultG=true, group=Quail Wink / Loan Float / Evoke Find, parameterName=Inverted Reflection Axis}
  n56 = MaterialFunctionCall() {materialFunction=&AxisAlignedFresnel}
  n57 = Power(
   Base=StaticSwitchParameter(True=Max(A=@n52, B=@n55), False=@n56.Fresnel) {group=Quail Wink / Loan Float / Evoke Find, parameterName=Use Reflection Vector}
   Exp=ScalarParameter() {defaultValue=2.6918, parameterName=Reflection Vector - Exponent}
  )
  n58 = NamedRerouteUsage()
  n59 = MaterialFunctionCall(Exponent=ScalarParameter() {defaultValue=2.6918, parameterName=Reflection Vector - Exponent}, Normal=PixelNormalWS()) {materialFunction=&AxisAlignedFresnel}
  n60 = Multiply(A=@n58, B=Multiply(A=OneMinus(None=@n59.Fresnel), B=NamedRerouteUsage()))
  n61 = ComponentMask(None=ReflectionVectorWS()) {b=true}
  n62 = Saturate(None=Power(
    Base=OneMinus(None=@n61)
    Exp=ScalarParameter() {defaultValue=0.654, group=Quail Wink / Loan Float / Evoke Find, parameterName=Reflection Vector - Y Exponent}
   ))
  n63 = Multiply(
   A=LinearInterpolate(
    A=ScalarParameter() {defaultValue=1, group=Quail Wink / Loan Float / Evoke Find, parameterName=Reflection Vector - Strength, sliderMin=18.9627}
    B=ScalarParameter() {group=Quail Wink / Loan Float / Evoke Find, parameterName=Edge Reflection - Strength}
    Alpha=NamedRerouteUsage()
   )
   B=Multiply(
    A=Saturate(None=@n57)
    B=StaticSwitchParameter(True=@n60, False=@n62) {group=Quail Wink / Loan Float / Evoke Find, parameterName=Evoke Find Reflection}
   )
  )
  n64 = StaticSwitchParameter(True=@n63, False=Saturate(None=@n63)) {group=Quail Wink / Loan Float / Evoke Find, parameterName=Evoke Find Reflection}
  n65 = ScalarParameter() {defaultValue=4, group=Quail Wink / Loan Float / Evoke Find, parameterName=Roughness - Base}
  n66 = ScalarParameter() {defaultValue=4, group=Quail Wink / Loan Float / Evoke Find, parameterName=Roughness - Edge}
  n67 = NamedRerouteUsage()
  n68 = Saturate(None=Multiply(A=LinearInterpolate(A=@n65, B=@n66, Alpha=@n67), B=NamedRerouteUsage()))
  n69 = Power(Base=NamedRerouteUsage(), Exp=ScalarParameter() {defaultValue=5.9459, group=Quail Wink / Loan Float / Evoke Find, parameterName=AO Exponent})
  NamedRerouteDeclaration(None=Multiply(A=Multiply(A=@n64, B=@n68), B=Saturate(None=@n69))) {name=Reflection Vector, nodeColor=<rgba 1 0.9844 0 1>}
  NamedRerouteDeclaration(None=@n21) {name=Reflection Edge, nodeColor=<rgba 1 0 0.0937 1>}
  NamedRerouteDeclaration(None=Constant()) {name=Pattern Clip Mask, nodeColor=<rgba 0.0312 1 0 1>}
  n70 = TextureSampleParameter2D!(
   UVs
   Bias=ScalarParameter() {defaultValue=-3, group=Quail Wink / Loan Float / Evoke Find, parameterName=Virtual Coords - Mip Bias}
   Apply View MipBias
  ) {
   group=Quail Wink / Loan Float / Evoke Find
   mipValueMode=TMVM_MipBias
   parameterName=FX Map
   texture=/BRCosmetics/Weapons/FORT_Melee/Pickaxe_LoanFloat/Textures/T_LoanFloat_Pickaxe_FX
  }
  n71 = StaticComponentMaskParameter(None=@n70.RGBA) {defaultR=true, group=Quail Wink / Loan Float / Evoke Find, parameterName=Pattern Mask}
  NamedRerouteDeclaration(None=@n71) {name=Non VC Mask, nodeColor=<rgba 1 0 0.8203 1>}
  NamedRerouteDeclaration(None=@n8.XYZ Texture) {name=Grayscale Gradient, nodeColor=<rgba 0.0547 0 1 1>}
  NamedRerouteDeclaration(None=ComponentMask(None=@n19.M) {g=true}) {name=Curvature?, nodeColor=<rgba 1 0 0.3047 1>}
  n72 = Constant() {r=2}
  n73 = Constant() {r=0.01}
  n74 = ComponentMask(None=Saturate(None=Fresnel(ExponentIn=@n72, BaseReflectFractionIn=@n73) {exponent=1})) {r=true}
  NamedRerouteDeclaration(None=@n74) {name=Curvature, nodeColor=<rgba 1 0 0.5391 1>}
  n75 = TextureObject() {samplerType=SAMPLERTYPE_LinearColor, texture=&T_PerlinNoise_Uncompressed}
  n76 = ScalarParameter() {defaultValue=150, parameterName=Param_3}
  n77 = Time()
  n78 = VectorParameter() {group=Quail Wink / Loan Float / Evoke Find, parameterName=Noise - Panning Offset, defaultValue=<rgba 0 0 -0.01 0>}
  n79 = Add(A=VertexInterpolator(VS=LocalPosition().XYZ).PS, B=Multiply(A=Multiply(A=@n77) {constB=0.5}, B=@n78.RGB))
  n80 = VertexInterpolator(VS=PreSkinnedNormal())
  n81 = ScalarParameter() {parameterName=Contrast}
  n82 = TextureCoordinate()
  n83 = ScalarParameter() {defaultValue=0.5, group=Quail Wink / Loan Float / Evoke Find, parameterName=Shining Noise - Tiling}
  n84 = VectorParameter() {group=Quail Wink / Loan Float / Evoke Find, parameterName=Shining Noise Offset, defaultValue=<rgba 0.01 0.05 0 0>}
  n85 = TextureSample(UVs=Panner(Coordinate=Multiply(A=@n82, B=@n83), Speed=@n84.RGB)) {
   samplerSource=SSM_Wrap_WorldGroupSettings
   samplerType=SAMPLERTYPE_LinearColor
   texture=&T_PerlinNoise_Uncompressed
  }
  n86 = LinearInterpolate(
   A=ComponentMask(None=MaterialFunctionCall(TextureObject=@n75, TextureSize=@n76, WorldPosition=@n79, World Space Normal=@n80.PS, ProjectionTransitionContrast=@n81) {materialFunction=&WorldAlignedTexture}.XYZ Texture) {r=true}
   B=@n85.R
   Alpha=ScalarParameter() {parameterName=WALI}
  )
  NamedRerouteDeclaration(None=@n86) {name=Shining Noise, nodeColor=<rgba 0.5703 1 0 1>}
  n87 = PixelNormalWS()
  NamedRerouteDeclaration(None=DotProduct(A=ReflectionVectorWS(CustomWorldNormal=@n87), B=CameraVectorWS())) {name=Camera Dot Reflection, nodeColor=<rgba 0.6406 0 1 1>}
  NamedRerouteDeclaration(None=ComponentMask(None=@n19.M) {r=true}) {name=AO, nodeColor=<rgba 0 0.2969 1 1>}
  NamedRerouteDeclaration(None=@n74) {name=Luminance Mask, nodeColor=<rgba 0.1953 0 1 1>}
  n88 = Transform(None=PixelNormalWS()) {transformSourceType=TRANSFORMSOURCE_World, transformType=TRANSFORM_Local}
  n89 = DotProduct(A=@n88, B=VectorParameter() {parameterName=Param_1, defaultValue=<rgba -1 0 0 1>}.RGB)
  n90 = DotProduct(A=@n88, B=VectorParameter() {parameterName=Param_2, defaultValue=<rgba 1 0 0 1>}.RGB)
  NamedRerouteDeclaration(None=Power(Base=Max(A=@n89, B=@n90), Exp=ScalarParameter() {defaultValue=1})) {name=Side Normals, nodeColor=<rgba 1 0 0.8203 1>}
  NamedRerouteDeclaration(None=ComponentMask(None=@n19.Tech Art Mask) {r=true}) {name=Blade Cell Noise, nodeColor=<rgba 0 0.2031 1 1>}
  NamedRerouteDeclaration(None=@n86) {name=Shining, nodeColor=<rgba 0.7344 0 1 1>}
  NamedRerouteDeclaration(None=@n14.Emissive Color) {name=Emissive In, nodeColor=<rgba 1 0 0.5391 1>}
  n91 = Power(
   Base=NamedRerouteUsage()
   Exp=ScalarParameter() {defaultValue=1, group=Quail Wink / Loan Float / Evoke Find, parameterName=Shining Noise - Exp}
  )
  n92 = Saturate(None=Multiply(
    A=Saturate(None=@n91)
    B=ScalarParameter() {defaultValue=0.5, group=Quail Wink / Loan Float / Evoke Find, parameterName=Shining Noise - Strength}
   ))
  n93 = NamedRerouteUsage()
  n94 = ScalarParameter() {defaultValue=1, group=Quail Wink / Loan Float / Evoke Find, parameterName=AO Strength}
  n95 = Multiply(A=TextureCoordinate(), B=ScalarParameter() {defaultValue=0.5, parameterName=B})
  n96 = AppendVector(A=ScalarParameter() {defaultValue=1, parameterName=B_1}, B=ScalarParameter() {parameterName=B_2})
  n97 = TextureSampleParameter2D!(
   UVs=MaterialFunctionCall(UVs=Add(A=@n95, B=@n96) {constB=0}, "Rotation Angle (0-1)"=ScalarParameter() {defaultValue=0.2, parameterName="Rotation Angle (0-1) (S)"}) {materialFunction=&CustomRotator}.Rotated Values
   Apply View MipBias
  ) {
   group=Quail Wink / Loan Float / Evoke Find
   parameterName=Gradient Color
   texture=/BRCosmetics/Weapons/FORT_Melee/Pickaxe_QuailWink/T_ColorShift_02_Gradient
  }
  n98 = MaterialFunctionCall!(
   Base=@n97.RGB
   Hue=Constant()
   Saturation=ScalarParameter() {defaultValue=0.9, group=Quail Wink / Loan Float / Evoke Find, parameterName=Saturation}
   Value
   Effect Amount
  ) {materialFunction=/BRCosmetics/Weapons/FORT_Melee/Pickaxe_QuailWink/MF_DM_Effect_Base_HSV}
  n99 = Multiply(
   A=Saturate(None=Power(Base=@n93, Exp=@n94))
   B=NamedRerouteDeclaration(None=NamedRerouteDeclaration(None=@n98.Result) {name=Gradient Texture, nodeColor=<rgba 0.6172 1 0 1>}) {name=Gradient + Blade Edge, nodeColor=<rgba 1 0 0.7031 1>}
  )
  n100 = Multiply(A=@n99, B=ScalarParameter() {defaultValue=2, group=Quail Wink / Loan Float / Evoke Find, parameterName=Darkening Intensity})
  n101 = Add(
   A=LocalPosition().Z
   B=ScalarParameter() {defaultValue=-10.7786, group=Quail Wink / Loan Float / Evoke Find, parameterName=Darkening Position}
  )
  n102 = ScalarParameter() {defaultValue=6.505, group=Quail Wink / Loan Float / Evoke Find, parameterName=Darkening Thickness}
  n103 = Add(A=@n101, B=@n102)
  n104 = Add(A=OneMinus(None=@n101), B=@n102)
  n105 = Saturate(None=Divide(A=Min(A=@n103, B=@n104), B=ScalarParameter() {defaultValue=5.2808, group=Quail Wink / Loan Float / Evoke Find, parameterName=Darkening Fade}))
  n106 = Fresnel(ExponentIn=ScalarParameter() {defaultValue=0.6, group=Quail Wink / Loan Float / Evoke Find, parameterName=Gradient Fresnel - Exp}) {baseReflectFraction=0}
  n107 = Saturate(None=Max(A=OneMinus(None=@n106), B=ScalarParameter() {group=Quail Wink / Loan Float / Evoke Find, parameterName=Gradient Fresnel - Min}))
  n108 = VectorParameter() {group=Quail Wink / Loan Float / Evoke Find, parameterName=Gradient Fresnel Color, defaultValue=<rgba 1 1 1 1>}
  n109 = Multiply(A=@n106, B=@n108.RGB)
  n110 = Multiply(
   A=Multiply(A=LinearInterpolate(A=@n99, B=@n100, Alpha=@n105), B=Max(A=@n107, B=@n109))
   B=ScalarParameter() {defaultValue=1, group=Quail Wink / Loan Float / Evoke Find, parameterName=Gradient Strength}
  )
  n111 = OneMinus(None=@n74)
  n112 = Add(
   A=Saturate(None=MaterialFunctionCall!(Blend=@n92, Base=@n110) {materialFunction=/Engine/Functions/Engine_MaterialFunctions03/Blends/Blend_ColorDodge}.Result)
   B=Multiply(A=NamedRerouteUsage(), B=NamedRerouteDeclaration(None=@n111) {name=Curvature Inverted, nodeColor=<rgba 0.9219 1 0 1>})
  )
  n113 = ScalarParameter() {defaultValue=0.7, group=Quail Wink / Loan Float / Evoke Find, parameterName=Curvature Value}
  NamedRerouteDeclaration(None=LinearInterpolate(A=@n112, B=Multiply(A=@n112, B=@n113), Alpha=Saturate(None=@n74))) {name=Effect, nodeColor=<rgba 0 0.8594 1 1>}
  NamedRerouteDeclaration(None=StaticComponentMaskParameter(None=@n19.Tech Art Mask) {defaultB=true, group=Quail Wink / Loan Float / Evoke Find, parameterName=Blade Mask}) {name=Blade Mask, nodeColor=<rgba 1 0.8203 0 1>}
  NamedRerouteDeclaration(None=Step(Y=ScalarParameter() {defaultValue=0.072, group=Quail Wink / Loan Float / Evoke Find, parameterName=Emissive Removal}, X=@n71)) {name=Pattern Mask, nodeColor=<rgba 0.5 0 1 1>}
  NamedRerouteUsage()
  NamedRerouteUsage()
  PixelNormalWS()
  PixelNormalWS()
  PixelNormalWS()
  Saturate(None=Power(Base=@n58, Exp=ScalarParameter() {defaultValue=0.1, parameterName=Luminance Mask - Exp}))
  ScalarParameter() {group=Quail Wink / Loan Float / Evoke Find, parameterName=SSS - Opacity}
  SetMaterialAttributes!(MaterialAttributes=@n18.MaterialAttributes, Emissive Color=ComponentMask(None=NamedRerouteUsage()) {b=true}) {attributeSetTypes [#1]}
  ShadingModel()
  StaticSwitchParameter(True=NamedRerouteUsage(), False=@n56.Fresnel) {group=Quail Wink / Loan Float / Evoke Find, parameterName=Evoke Find Reflection}
  TextureSampleParameter2D() {
   group=Quail Wink / Loan Float / Evoke Find
   parameterName=Edge Mask
   samplerType=SAMPLERTYPE_Grayscale
   texture=/BRCosmetics/Weapons/FORT_Melee/Pickaxe_QuailWink/T_EdgeBlade_Mask
  }
  TextureSampleParameter2D() {
   group=Quail Wink / Loan Float / Evoke Find
   parameterName=FX2
   texture=/BRCosmetics/Weapons/FORT_Melee/Pickaxe_QuailWink/Textures/T_QuailWink_FX3
  }
  TextureSample(UVs=Multiply(
    A=TextureCoordinate()
    B=ScalarParameter() {defaultValue=1, group=Quail Wink / Loan Float / Evoke Find, parameterName=Noise Modulation - Tiling}
   )) {
   samplerSource=SSM_Wrap_WorldGroupSettings
   samplerType=SAMPLERTYPE_LinearGrayscale
   texture=&T_ColorShift_02_Mask
  }
 ]
}
